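# Animation Blueprint for the M_LRG_Bob cosmetic body, adding secondary motion to a pose copied from the owner's skeletal mesh.
# On initialization it caches that mesh in MeshToCopy for Copy Pose From Mesh.
# A chain of Anim Dynamics nodes then simulates pouches, wrench, arm sign, straps and sleeves before output.
# A Pose Driver and curve-based buckle Bone Driven Controllers, plus two frozen sequence players, are left unconnected to the output.
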AnimBlueprint /BRCosmetics/Characters/Player/Male/Large/Bodies/M_LRG_Bob/Meshes/M_LRG_Bob_AnimBP
target_skeleton=&/M_LRG_Bob_Skeleton
graphs [
 AnimGraph {nodes [
   n1 = Pose Driver - Source: spine_05!(SourcePose:Pose Link Structure) -> (Pose:Pose Link Structure) {
    is=title AnimGraphNode_PoseDriver
    poseAsset=/BRCosmetics/Characters/Player/Male/Large/Bodies/M_LRG_Bob/Animation/M_LRG_Bob_PoseAsset_Simple
    evalSpaceBone {boneName=pelvis}
    onlyDriveBones [{boneName=buckle_bk_l}, {boneName=buckle_bk_r}]
    poseTargets [
     {
      bApplyCustomCurve=false
      bIsHidden=false
      distanceMethod=DefaultMethod
      drivenName=Pose_1
      functionType=DefaultFunction
      targetScale=1
      boneTransforms [{targetRotation=<rot 0 75.4227 0>, targetTranslation=<31.1015 18.2349 0>}]
      customCurve {defaultValue=3.4028234663852886e+38, postInfinityExtrap=RCCE_Constant, preInfinityExtrap=RCCE_Constant, keys []}
      targetRotation=<rot 0 0 0>
     }
     {
      bApplyCustomCurve=false
      bIsHidden=false
      distanceMethod=DefaultMethod
      drivenName=Pose_2
      functionType=DefaultFunction
      targetScale=1
      boneTransforms [{targetRotation=<rot 0 -39.9507 0>, targetTranslation=<35.6464 -13.7446 0>}]
      customCurve {defaultValue=3.4028234663852886e+38, postInfinityExtrap=RCCE_Constant, preInfinityExtrap=RCCE_Constant, keys []}
      targetRotation=<rot 0 0 0>
     }
     {
      bApplyCustomCurve=false
      bIsHidden=false
      distanceMethod=DefaultMethod
      drivenName=Pose_3
      functionType=DefaultFunction
      targetScale=1
      boneTransforms [{targetRotation=<rot -0.3212 -6.8408 -71.9529>, targetTranslation=<38.6119 -4.1908 -0.3727>}]
      customCurve {defaultValue=3.4028234663852886e+38, postInfinityExtrap=RCCE_Constant, preInfinityExtrap=RCCE_Constant, keys []}
      targetRotation=<rot 0 0 0>
     }
     {
      bApplyCustomCurve=false
      bIsHidden=false
      distanceMethod=DefaultMethod
      drivenName=Pose_4
      functionType=DefaultFunction
      targetScale=1
      boneTransforms [{targetRotation=<rot 0.2063 -6.7579 52.8192>, targetTranslation=<38.6113 -4.2062 0.2724>}]
      customCurve {defaultValue=3.4028234663852886e+38, postInfinityExtrap=RCCE_Constant, preInfinityExtrap=RCCE_Constant, keys []}
      targetRotation=<rot 0 0 0>
     }
     {
      bApplyCustomCurve=false
      bIsHidden=false
      distanceMethod=DefaultMethod
      drivenName=Pose_5
      functionType=DefaultFunction
      targetScale=1
      boneTransforms [{targetRotation=<rot -49.9832 -6.9755 0.2982>, targetTranslation=<34.656 -3.8437 -14.2899>}]
      customCurve {defaultValue=3.4028234663852886e+38, postInfinityExtrap=RCCE_Constant, preInfinityExtrap=RCCE_Constant, keys []}
      targetRotation=<rot 0 0 0>
     }
     {
      bApplyCustomCurve=false
      bIsHidden=false
      distanceMethod=DefaultMethod
      drivenName=Pose_6
      functionType=DefaultFunction
      targetScale=1
      boneTransforms [{targetRotation=<rot 49.9832 -6.9755 -0.2981>, targetTranslation=<34.656 -3.8437 14.2899>}]
      customCurve {defaultValue=3.4028234663852886e+38, postInfinityExtrap=RCCE_Constant, preInfinityExtrap=RCCE_Constant, keys []}
      targetRotation=<rot 0 0 0>
     }
     {
      bApplyCustomCurve=false
      bIsHidden=false
      distanceMethod=DefaultMethod
      drivenName=Pose_7
      functionType=DefaultFunction
      targetScale=1
      boneTransforms [{targetRotation=<rot -4.2788 39.4704 -32.4214>, targetTranslation=<35.8096 10.8376 -0.1441>}]
      customCurve {defaultValue=3.4028234663852886e+38, postInfinityExtrap=RCCE_Constant, preInfinityExtrap=RCCE_Constant, keys []}
      targetRotation=<rot 0 0 0>
     }
     {
      bApplyCustomCurve=false
      bIsHidden=false
      distanceMethod=DefaultMethod
      drivenName=Pose_8
      functionType=DefaultFunction
      targetScale=1
      boneTransforms [{targetRotation=<rot 15.8851 -1.1331 -50.9775>, targetTranslation=<38.5331 1.992 2.6177>}]
      customCurve {defaultValue=3.4028234663852886e+38, postInfinityExtrap=RCCE_Constant, preInfinityExtrap=RCCE_Constant, keys []}
      targetRotation=<rot 0 0 0>
     }
     {
      bApplyCustomCurve=false
      bIsHidden=false
      distanceMethod=DefaultMethod
      drivenName=Pose_9
      functionType=DefaultFunction
      targetScale=1
      boneTransforms [{targetRotation=<rot -27.6275 -7.1402 99.9755>, targetTranslation=<37.4202 3.5014 -10.2413>}]
      customCurve {defaultValue=3.4028234663852886e+38, postInfinityExtrap=RCCE_Constant, preInfinityExtrap=RCCE_Constant, keys []}
      targetRotation=<rot 0 0 0>
     }
     {
      bApplyCustomCurve=false
      bIsHidden=false
      distanceMethod=DefaultMethod
      drivenName=Pose_10
      functionType=DefaultFunction
      targetScale=1
      boneTransforms [{targetRotation=<rot 1.4131 -7.8532 92.7929>, targetTranslation=<39.2136 1.3517 -2.6739>}]
      customCurve {defaultValue=3.4028234663852886e+38, postInfinityExtrap=RCCE_Constant, preInfinityExtrap=RCCE_Constant, keys []}
      targetRotation=<rot 0 0 0>
     }
    ]
    rBFParams {distanceMethod=Euclidean, function=Linear, radius=48}
    sourceBones [{boneName=spine_05}]
   }
   n2 = "buckle_bk_l.translateY = curve(spine_03.rotateZ) - Bone Driven Controller"!(ComponentPose:Component Space Pose Link Structure=Local To Component(LocalPose=@n1), Alpha:float=1) -> (Pose:Component Space Pose Link Structure) {
    is=title AnimGraphNode_BoneDrivenController
    bAffectTargetTranslationY=true
    drivingCurve=&/M_LRG_Bob_StrapCurve
    multiplier=2.5
    sourceComponent=RotationZ
    alphaBoolBlend {blendOption=Linear}
    sourceBone {boneName=spine_03}
    targetBone {boneName=buckle_bk_l}
   }
   "buckle_bk_r.translateY = curve(spine_03.rotateZ) - Bone Driven Controller"!(ComponentPose:Component Space Pose Link Structure=@n2, Alpha:float=1) -> (Pose:Component Space Pose Link Structure) {
    is=title AnimGraphNode_BoneDrivenController
    bAffectTargetTranslationY=true
    drivingCurve=&M_LRG_Bob_StrapCurve
    multiplier=2.5
    sourceComponent=RotationZ
    alphaBoolBlend {blendOption=Linear}
    sourceBone {boneName=spine_03}
    targetBone {boneName=buckle_bk_r}
   }
   n3 = Get MeshToCopy() -> (MeshToCopy:Skeletal Mesh Component Object Reference)
   n4 = Anim Dynamics - Bone: dyn_pouch_l(ComponentPose=Local To Component(LocalPose=Copy Pose From Mesh(SourceMeshComponent=@n3)), Alpha=1, ExternalForce="0,0,0") {
    angularDampingOverride=0.99
    angularSpringConstant=1500
    bAngularSpring=true
    bGravityOverrideInSimSpace=true
    bOverrideAngularDamping=true
    numSolverIterationsPostUpdate=4
    numSolverIterationsPreUpdate=16
    simulationSpace=Actor
    alphaBoolBlend {blendOption=Linear}
    boundBone {boneName=dyn_pouch_l}
    physicsBodyDefinitions [{
      sphereCollisionRadius=0
      boundBone {boneName=dyn_pouch_l}
      boxExtents=<20 20 10>
      constraintSetup {twistAxis=AxisZ, angularLimitsMax=<30 40 5>, angularLimitsMin=<-30 -5 -5>, angularTarget=<1 0 0>}
      localJointOffset=<0 0 -3>
     }]
   }
   n5 = Anim Dynamics - Bone: dyn_pouch_flap_l(ComponentPose=@n4, Alpha=1, ExternalForce="0,0,0") {
    angularDampingOverride=0.99
    angularSpringConstant=5000
    bAngularSpring=true
    bGravityOverrideInSimSpace=true
    bOverrideAngularDamping=true
    numSolverIterationsPostUpdate=4
    numSolverIterationsPreUpdate=16
    simulationSpace=Actor
    alphaBoolBlend {blendOption=Linear}
    boundBone {boneName=dyn_pouch_flap_l}
    physicsBodyDefinitions [{
      sphereCollisionRadius=0
      boundBone {boneName=dyn_pouch_flap_l}
      boxExtents=<40 40 10>
      constraintSetup {twistAxis=AxisZ, angularLimitsMax=<0 40 0>, angularLimitsMin=<0 0 0>, angularTarget=<1 0 0>}
      localJointOffset=<0 0 -3>
     }]
   }
   n6 = Anim Dynamics - Bone: dyn_pouch_f(ComponentPose=@n5, Alpha=1, ExternalForce="0,0,0") {
    angularDampingOverride=0.99
    angularSpringConstant=1500
    bAngularSpring=true
    bGravityOverrideInSimSpace=true
    bOverrideAngularDamping=true
    numSolverIterationsPostUpdate=4
    numSolverIterationsPreUpdate=16
    simulationSpace=Actor
    alphaBoolBlend {blendOption=Linear}
    boundBone {boneName=dyn_pouch_f}
    physicsBodyDefinitions [{
      sphereCollisionRadius=0
      boundBone {boneName=dyn_pouch_f}
      boxExtents=<20 10 20>
      constraintSetup {angularLimitsMax=<5 20 20>, angularLimitsMin=<-5 -20 -15>, angularTarget=<1 0 0>}
      localJointOffset=<0 -3 0>
     }]
   }
   n7 = Anim Dynamics - Bone: dyn_wrench(ComponentPose=@n6, Alpha=1, ExternalForce="0,0,0") {
    angularDampingOverride=0.99
    angularSpringConstant=1500
    bAngularSpring=true
    bGravityOverrideInSimSpace=true
    bOverrideAngularDamping=true
    numSolverIterationsPostUpdate=4
    numSolverIterationsPreUpdate=16
    simulationSpace=Actor
    alphaBoolBlend {blendOption=Linear}
    boundBone {boneName=dyn_wrench}
    physicsBodyDefinitions [{
      sphereCollisionRadius=0
      boundBone {boneName=dyn_wrench}
      boxExtents=<20 20 10>
      constraintSetup {twistAxis=AxisZ, angularLimitsMax=<30 40 5>, angularLimitsMin=<-30 0 -5>, angularTarget=<1 0 0>}
      localJointOffset=<0 0 -3>
     }]
   }
   n8 = Anim Dynamics - Bone: dyn_arm_sign(ComponentPose=@n7, Alpha=1, ExternalForce="0,0,0") {
    angularBiasOverride=1
    bGravityOverrideInSimSpace=true
    bOverrideAngularBias=true
    numSolverIterationsPostUpdate=4
    numSolverIterationsPreUpdate=16
    alphaBoolBlend {blendOption=Linear}
    boundBone {boneName=dyn_arm_sign}
    physicsBodyDefinitions [{
      sphereCollisionRadius=0
      boundBone {boneName=dyn_arm_sign}
      boxExtents=<10 10 10>
      constraintSetup {angularLimitsMax=<3 0 0>, angularLimitsMin=<-3 0 0>, angularTarget=<0 0 0>}
      localJointOffset=<0 0 0>
     }]
   }
   n9 = Anim Dynamics - Bone: dyn_strap_l(ComponentPose=@n8, Alpha=1, ExternalForce="0,0,0") {
    angularBiasOverride=1
    bGravityOverrideInSimSpace=true
    bOverrideAngularBias=true
    bUsePlanarLimit=false
    gravityScale=0.1
    numSolverIterationsPostUpdate=4
    numSolverIterationsPreUpdate=16
    alphaBoolBlend {blendOption=Linear}
    boundBone {boneName=dyn_strap_l}
    physicsBodyDefinitions [{
      sphereCollisionRadius=0
      boundBone {boneName=dyn_strap_l}
      boxExtents=<5 5 5>
      constraintSetup {angularTargetAxis=AxisY, twistAxis=AxisY, angularLimitsMax=<0 0 0>, angularLimitsMin=<0 0 0>, angularTarget=<0 0 0>, linearAxesMax=<0 0 1>}
      localJointOffset=<0 0 -5>
     }]
   }
   n10 = Anim Dynamics - Bone: dyn_strap_r(ComponentPose=@n9, Alpha=1, ExternalForce="0,0,0") {
    angularBiasOverride=1
    bGravityOverrideInSimSpace=true
    bOverrideAngularBias=true
    bUsePlanarLimit=false
    gravityScale=0.1
    numSolverIterationsPostUpdate=4
    numSolverIterationsPreUpdate=16
    alphaBoolBlend {blendOption=Linear}
    boundBone {boneName=dyn_strap_r}
    physicsBodyDefinitions [{
      sphereCollisionRadius=0
      boundBone {boneName=dyn_strap_r}
      boxExtents=<5 5 5>
      constraintSetup {angularTargetAxis=AxisY, twistAxis=AxisY, angularLimitsMax=<0 0 0>, angularLimitsMin=<0 0 0>, angularTarget=<0 0 0>, linearAxesMax=<0 0 1>}
      localJointOffset=<0 0 -5>
     }]
   }
   n11 = Anim Dynamics - Bone: dyn_sleeve_r(ComponentPose=@n10, Alpha=.8, ExternalForce="0,0,0") {
    angularBiasOverride=1
    bGravityOverrideInSimSpace=true
    bOverrideAngularBias=true
    bUsePlanarLimit=false
    gravityScale=0.1
    numSolverIterationsPostUpdate=4
    numSolverIterationsPreUpdate=16
    alphaBoolBlend {blendOption=Linear}
    boundBone {boneName=dyn_sleeve_r}
    physicsBodyDefinitions [{
      sphereCollisionRadius=0
      boundBone {boneName=dyn_sleeve_r}
      boxExtents=<5 5 5>
      constraintSetup {angularLimitsMax=<0 0 0>, angularLimitsMin=<0 0 0>, angularTarget=<0 0 0>, linearAxesMax=<2.5 0.001 0.001>, linearAxesMin=<0 -0.001 -0.001>}
      localJointOffset=<-5 0 0>
     }]
   }
   n12 = Anim Dynamics - Bone: dyn_sleeve_l(ComponentPose=@n11, Alpha=.8, ExternalForce="0,0,0") {
    angularBiasOverride=1
    bGravityOverrideInSimSpace=true
    bOverrideAngularBias=true
    bUsePlanarLimit=false
    gravityScale=0.1
    numSolverIterationsPostUpdate=4
    numSolverIterationsPreUpdate=16
    alphaBoolBlend {blendOption=Linear}
    boundBone {boneName=dyn_sleeve_l}
    physicsBodyDefinitions [{
      sphereCollisionRadius=0
      boundBone {boneName=dyn_sleeve_l}
      boxExtents=<5 5 5>
      constraintSetup {angularLimitsMax=<0 0 0>, angularLimitsMin=<0 0 0>, angularTarget=<0 0 0>, linearAxesMax=<2.5 0.001 0.001>, linearAxesMin=<0 -0.001 -0.001>}
      localJointOffset=<-5 0 0>
     }]
   }
   Output Pose(Result=Component To Local(ComponentPose=@n12)) {layerGroup=None}
   Sequence Player 'InGame_Movements'() {playRate=0, sequence=/BRCosmetics/Characters/Player/Male/Large/Bodies/M_LRG_Bob/Animation/InGame_Movements, startPosition=5.9}
   Sequence Player 'InGame_Movements_01'() {playRate=0, sequence=/BRCosmetics/Characters/Player/Male/Large/Bodies/M_LRG_Bob/Animation/InGame_Movements_01, startPosition=10.278}
  ]}
 EventGraph {nodes [
   Event BlueprintInitializeAnimation()
   > Set MeshToCopy(MeshToCopy:Skeletal Mesh Component Object Reference=GetOwnerSkeletalMeshComponent()) -> (Output_Get:Skeletal Mesh Component Object Reference)
   TryGetPawnOwner()
   Event BlueprintUpdateAnimation()
  ]}
]
variables [MeshToCopy {type=SkeletalMeshComponent}]
